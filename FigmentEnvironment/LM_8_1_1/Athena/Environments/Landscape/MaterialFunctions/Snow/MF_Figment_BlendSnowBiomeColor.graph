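# Figment material function that optionally tints snow material attributes with the static mesh biome color, switched by UseBiomeColor.
# The Snow input, defaulting to MF_Athena_Snow_01, is multiplied by that color through MF_Athena_Material_BaseMultiply.
# A world-XY SphereMask blends back to untinted snow inside its parameterized radius, leaving the tint outside.
# The mask center uses BiomeColorMaskPos.R for both X and Y, so G is ignored.
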MaterialFunction /FigmentEnvironment/LM_8_1_1/Athena/Environments/Landscape/MaterialFunctions/Snow/MF_Figment_BlendSnowBiomeColor
graph {
 properties {bExposeToLibrary=true, libraryCategoriesText [Figment]}
 nodes [
  n1 = FunctionInput(Preview=MaterialFunctionCall!(UseLandscapeCoordsOnSwitch) {materialFunction=&/MF_Athena_Snow_01}.Material) {bUsePreviewValueAsDefault=true, inputName=Snow, inputType=FunctionInput_MaterialAttributes}
  n2 = MaterialFunctionCall!(
   In=@n1
   Color=MaterialFunctionCall!(Material) {materialFunction=/FigmentEnvironment/LM_8_1_1/Athena/Environments/Materials/MF_StaticMeshBiomeColor}.Color
  ) {materialFunction=/FigmentEnvironment/LM_8_1_1/Athena/Environments/Landscape/MaterialFunctions/MF_Athena_Material_BaseMultiply}
  n3 = VectorParameter() {group=Snow Biome Color, parameterName=BiomeColorMaskPos, defaultValue=<rgba 0 0 0 1>}
  n4 = SphereMask(
   A=WorldPosition().XY
   B=AppendVector(A=@n3.R, B=@n3.R)
   Radius=ScalarParameter() {group=Snow Biome Color, parameterName=BiomeColorMaskRadius}
   Hardness=ScalarParameter() {group=Snow Biome Color, parameterName=BiomeColorMaskHardness}
  )
  FunctionOutput(None=StaticSwitchParameter(True=BlendMaterialAttributes(A=@n2.Result, B=@n1, Alpha=@n4), False=@n1) {group=Snow Biome Color, parameterName=UseBiomeColor, sortPriority=1})
 ]
}
